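# Niagara sprite system for a short, faint steam puff, named for the MoonFlax heavy sniper's reload.
# One Steam emitter fires two timed bursts of short-lived grey sprites from a local offset, drifting along negative X as they grow and rotate.
# Particles follow source movement with decreasing strength, while curves drive a material parameter, low alpha fade and a bright-to-normal color falloff.
# Continuous spawning, SubUV animation and upward acceleration are disabled.
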NiagaraSystem /MoonFlaxWeaponGameplay/Items/Weapons/Sniper/FX/NS_MoonFlax_HeavySniper_ReloadSteam
dependencies {
 usedDataInterfaces [
  &NiagaraDataInterfaceCurve
  &NiagaraDataInterfaceSpriteRendererInfo
  &NiagaraDataInterfaceColorCurve
 ]
 usedDynamicInputs []
 usedModules [&SystemState]
}
emitters {Steam {
  emitterUpdateScript [
   &EmitterState()
   &SpawnBurst_Instantaneous(Spawn Count=3, Use Spawn Probability=false, UseLoopCountLimit=false) {module=SpawnBurst_Instantaneous001}
   &SpawnBurst_Instantaneous(Spawn Time=0.1, Use Spawn Probability=false, UseLoopCountLimit=false)
   &SpawnRate(SpawnRate=5, Use Spawn Probability=false) {enabled=false}
  ]
  particleSpawnScript [
   &InitializeParticle(
    Color=<rgba 0.5624 0.5469 0.5459 1>
    Lifetime Max=0.5
    Lifetime Min=0.3
    Lifetime Mode=::NewEnumerator1
    Position Offset=<10 5 0>
    Position Offset Coordinate Space=::Local
    Sprite Rotation Mode=::NewEnumerator2
    Sprite Size Mode=::NewEnumerator1
    Uniform Sprite Size Max=50
    Uniform Sprite Size Min=35
    UsePositionOffset=true
    Write Material Random=true
   )
   &AddVelocity(Velocity=<-50 0 0>)
  ]
  particleUpdateScript [
   &ParticleState()
   &SpriteRotationRate(Rotation Rate=&RandomRangeFloat(Minimum=-2, Maximum=2))
   &ScaleSpriteSize(Uniform Curve Sprite Scale=DataInterface() {propertyValues {Curve {keys [
        {interpMode=RCIM_Cubic, leaveTangentWeight=0.083, tangentMode=RCTM_User, time=0, value=0.25}
        {arriveTangent=1.1625, interpMode=RCIM_Cubic, leaveTangent=1.1625, tangentMode=RCTM_User, time=0.15, value=0.45}
        <at 1 1 RCIM_Cubic>
       ]}}})
   &SubUVAnimation(
    Direct Frame Index:NiagaraFloat=&FloatFromCurve(FloatCurve=DataInterface() {propertyValues {Curve {keys [<at 0 20>,<at 1 63>]}}})
    "Sprite Renderer "=DataInterface() {propertyValues {SpriteRenderer=&self:Steam_0.NiagaraSpriteRendererProperties_0}}
    SubUV Animation Mode:ENiagara_SubUVLookupModeV2=::NewEnumerator5
   ) {enabled=false}
   &AccelerationForce(Acceleration=<0 0 100>) {enabled=false}
   &CameraOffset(Camera Offset Amount=10)
   &Drag(Drag=0, Use Linear Drag=true, Use Rotational Drag=false)
   &InheritSourceMovement(
    Applied Position Scale=&Multiply_VectorByFloat(
     Vector=<1 1 1>
     Float=&FloatFromCurve(FloatCurve=DataInterface() {propertyValues {Curve {keys [{interpMode=RCIM_Cubic, tangentMode=RCTM_User, time=0, value=1}, {interpMode=RCIM_Cubic, tangentMode=RCTM_User, time=1, value=0.6}]}}})
    )
    ApplyForce=false
    ApplyPosition=true
    ApplyPositionFalloff=false
   )
   &SolveForcesAndVelocity(Clamp Velocity=false, Limit Acceleration=false)
   &DynamicMaterialParameters(
    Index 0 Param 1=&FloatFromCurve(FloatCurve=DataInterface() {propertyValues {Curve {keys [<at 0 1>,<at 1 0>]}}})
    Index 0 Param 2=1
    Index 0 Param 4=1
    Module.Index0Param0Write=true
    Module.Index0Param1Write=true
    Module.Index0Param2Write=false
    Module.Index0Param3Write=true
   )
   &ScaleColor(
    Scale Alpha=&FloatFromCurve(FloatCurve=DataInterface() {propertyValues {Curve {keys [
         {arriveTangent=0.1538, interpMode=RCIM_Cubic, leaveTangent=0.1538, tangentMode=RCTM_User, time=0, value=0}
         {
          arriveTangent=0.0082
          arriveTangentWeight=0.0503
          interpMode=RCIM_Cubic
          leaveTangent=0.0082
          leaveTangentWeight=0.2515
          tangentMode=RCTM_User
          time=0.15
          value=0.1
         }
         {
          arriveTangent=0.0082
          arriveTangentWeight=0.2517
          interpMode=RCIM_Cubic
          leaveTangent=0.0082
          leaveTangentWeight=0.0336
          tangentMode=RCTM_User
          time=0.7
          value=0.1
         }
         {arriveTangent=-0.0117, interpMode=RCIM_Cubic, leaveTangent=-0.0117, tangentMode=RCTM_User, time=1, value=0}
        ]}}})
    Scale RGB=&MakeVectorFromLinearColorRGB(Color=&ColorFromCurve(ColorCurve=DataInterface() {propertyValues {
        AlphaCurve {keys []}
        BlueCurve {keys [<at 0 5 RCIM_Cubic>,<at 0.5 1 RCIM_Cubic>]}
        GreenCurve {keys [<at 0 4.9186 RCIM_Cubic>,<at 0.5 1 RCIM_Cubic>]}
        RedCurve {keys [<at 0 4.8697 RCIM_Cubic>,<at 0.5 1 RCIM_Cubic>]}
       }}))
    ScaleA=true
    ScaleRGB=true
   )
  ]
  rendererClasses [&NiagaraSpriteRendererProperties]
  renderers [{rendererClass=&NiagaraSpriteRendererProperties}]
 }}
summary {
 systemName=NS_MoonFlax_HeavySniper_ReloadSteam
 emitters [{emitterName=Steam, rendererClasses [&NiagaraSpriteRendererProperties]}]
 userVariables []
}
